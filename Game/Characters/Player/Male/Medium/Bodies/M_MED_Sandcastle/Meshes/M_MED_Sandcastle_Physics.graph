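# Physics asset for the Sandcastle medium male player body, covering a three-segment dynamic skirt plus the pelvis and both thighs.
# Box bodies form the skirt chain, which hangs from a tiny capsule master bone constrained to the pelvis with free rotation.
# Skirt joints allow a wide swing on one axis and little on the other or in twist.
# Pelvis and thigh capsules approximate the lower body, with thighs rotating freely.
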PhysicsAsset /Game/Characters/Player/Male/Medium/Bodies/M_MED_Sandcastle/Meshes/M_MED_Sandcastle_Physics
bodies {
 dyn_skirt_01 [{shapeType=Box, extentX=3.8118, extentY=9.3237, extentZ=22.3971, center=<2.6337 0 0>, rotation=<rot 0 -90.0004 0>}]
 dyn_skirt_01_MSTR [{shapeType=Capsule, length=0.01, radius=0.01}]
 dyn_skirt_02 [{shapeType=Box, extentX=4.4917, extentY=9.3237, extentZ=22.3971, center=<2.6337 0 0>, rotation=<rot 0 -90.0004 0>}]
 dyn_skirt_03 [{shapeType=Box, extentX=4.4917, extentY=9.3237, extentZ=22.3971, center=<2.6337 0 0>, rotation=<rot 0 -90.0004 0>}]
 pelvis [{shapeType=Capsule, length=19.7981, radius=12.0119, center=<-3.9988 -1.1529 -0.0103>, rotation=<rot 0 178.395 -179.6674>}]
 thigh_l [{shapeType=Capsule, length=21.9763, radius=10.2394, center=<-13.8914 0.0001 0>, rotation=<rot 90 0.0297 0.0298>}]
 thigh_r [{shapeType=Capsule, length=21.9763, radius=10.2394, center=<13.8914 -0.0001 0>, rotation=<rot 90 -0.0132 179.9869>}]
}
constraints [
 {bone1Name=dyn_skirt_01, bone2Name=dyn_skirt_01_MSTR, swing1LimitDegrees=45, swing2LimitDegrees=5, twistLimitDegrees=5}
 {bone1Name=dyn_skirt_02, bone2Name=dyn_skirt_01, swing1LimitDegrees=45, swing2LimitDegrees=5, twistLimitDegrees=5}
 {bone1Name=dyn_skirt_03, bone2Name=dyn_skirt_02, swing1LimitDegrees=45, swing2LimitDegrees=5, twistLimitDegrees=5}
 {
  bone1Name=thigh_l
  bone2Name=pelvis
  swing1LimitDegrees=45
  swing1Motion=Free
  swing2LimitDegrees=45
  swing2Motion=Free
  twistLimitDegrees=45
  twistMotion=Free
 }
 {
  bone1Name=thigh_r
  bone2Name=pelvis
  swing1LimitDegrees=45
  swing1Motion=Free
  swing2LimitDegrees=45
  swing2Motion=Free
  twistLimitDegrees=45
  twistMotion=Free
 }
 {
  bone1Name=dyn_skirt_01_MSTR
  bone2Name=pelvis
  swing1LimitDegrees=45
  swing1Motion=Free
  swing2LimitDegrees=45
  swing2Motion=Free
  twistLimitDegrees=45
  twistMotion=Free
 }
]
