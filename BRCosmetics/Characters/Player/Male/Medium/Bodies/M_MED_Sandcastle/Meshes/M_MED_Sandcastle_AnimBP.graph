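# Post-process style AnimBlueprint for the M_MED_Sandcastle player body, adding procedural muscle deformation and physics to a pose copied from the attached parent mesh.
# Weighted multi-bone constraints place pectoral aim bones, then constraints and Look Ats orient pectorals, scapulae and lats.
# Additive bicep and tricep transforms follow per-arm length alphas, feeding a pelvis-space RigidBody before output.
# Skirt AnimDynamics, bicep roll and earlier duplicate chains are unconnected to the output.
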AnimBlueprint /BRCosmetics/Characters/Player/Male/Medium/Bodies/M_MED_Sandcastle/Meshes/M_MED_Sandcastle_AnimBP
target_skeleton=/Game/Characters/Player/Male/Medium/Bodies/M_MED_Sandcastle/Meshes/M_MED_Sandcastle_Skeleton
graphs [
 AnimGraph {nodes [
   n1 = Anim Dynamics - Bone: dyn_skirt_01(Alpha=1, ExternalForce="0,0,0") {
    angularDampingOverride=1
    angularSpringConstant=1500
    bAngularSpring=true
    bGravityOverrideInSimSpace=true
    bOverrideAngularDamping=true
    gravityScale=0.4
    simulationSpace=RootRelative
    alphaBoolBlend {blendOption=Linear}
    boundBone {boneName=dyn_skirt_01}
    componentLinearAccScale=<0.2 0.2 0.3>
    componentLinearVelScale=<0.2 0.2 0.3>
    physicsBodyDefinitions [{
      sphereCollisionRadius=0
      boundBone {boneName=dyn_skirt_01}
      boxExtents=<10 20 20>
      constraintSetup {angularLimitsMax=<0 0 -5>, angularLimitsMin=<0 0 -45>, angularTarget=<0 0 0>}
      localJointOffset=<10 0 0>
     }]
   }
   n2 = Anim Dynamics - Bone: dyn_skirt_02(ComponentPose=@n1, Alpha=1, ExternalForce="0,0,0") {
    angularDampingOverride=1
    angularSpringConstant=1500
    bAngularSpring=true
    bGravityOverrideInSimSpace=true
    bOverrideAngularDamping=true
    gravityScale=0.4
    simulationSpace=RootRelative
    alphaBoolBlend {blendOption=Linear}
    boundBone {boneName=dyn_skirt_02}
    componentLinearAccScale=<0.2 0.2 0.3>
    componentLinearVelScale=<0.2 0.2 0.3>
    physicsBodyDefinitions [{
      sphereCollisionRadius=0
      boundBone {boneName=dyn_skirt_02}
      boxExtents=<10 20 20>
      constraintSetup {angularLimitsMax=<0 0 0>, angularLimitsMin=<0 0 -45>, angularTarget=<0 0 0>}
      localJointOffset=<10 0 0>
     }]
   }
   Anim Dynamics - Bone: dyn_skirt_03(ComponentPose=@n2, Alpha=1, ExternalForce="0,0,0") {
    angularDampingOverride=1
    angularSpringConstant=1500
    bAngularSpring=true
    bGravityOverrideInSimSpace=true
    bOverrideAngularDamping=true
    gravityScale=0.4
    simulationSpace=RootRelative
    alphaBoolBlend {blendOption=Linear}
    boundBone {boneName=dyn_skirt_03}
    componentLinearAccScale=<0.2 0.2 0.3>
    componentLinearVelScale=<0.2 0.2 0.3>
    physicsBodyDefinitions [{
      sphereCollisionRadius=0
      boundBone {boneName=dyn_skirt_03}
      boxExtents=<10 20 20>
      constraintSetup {angularLimitsMax=<0 0 10>, angularLimitsMin=<0 0 -45>, angularTarget=<0 0 0>}
      localJointOffset=<10 0 0>
     }]
   }
   n3 = Get BicepLeftLengthAlpha() -> (BicepLeftLengthAlpha:float)
   n4 = "Transform (Modify) Bone - Bone: bicep_02_l"(Alpha=@n3, Translation="-6,-0.000,-0.000", Rotation="0,0,0", Scale="1,1,1") {
    rotationMode=BMM_Additive
    rotationSpace=BCS_BoneSpace
    scaleMode=BMM_Additive
    scaleSpace=BCS_BoneSpace
    translationMode=BMM_Additive
    translationSpace=BCS_BoneSpace
    alphaBoolBlend {blendOption=Linear}
    boneToModify {boneName=bicep_02_l}
    translation {x=-6}
   }
   n5 = "Transform (Modify) Bone - Bone: bicep_01_l"(ComponentPose=@n4, Alpha=@n3, Translation="-0.000,0.000,2.343", Rotation="0,0,0", Scale="1.000,1.500,1.250") {
    rotationMode=BMM_Additive
    rotationSpace=BCS_BoneSpace
    scaleMode=BMM_Additive
    scaleSpace=BCS_BoneSpace
    translationMode=BMM_Additive
    translationSpace=BCS_BoneSpace
    alphaBoolBlend {blendOption=Linear}
    boneToModify {boneName=bicep_01_l}
    scale {y=1.5, z=1.25}
    translation {z=2.343}
   }
   n6 = "Transform (Modify) Bone - Bone: tricep_02_l"(ComponentPose=@n5, Alpha=@n3, Translation="-6,-0.000,-0.000", Rotation="0,0,0", Scale="1,1,1") {
    rotationMode=BMM_Additive
    rotationSpace=BCS_BoneSpace
    scaleMode=BMM_Additive
    scaleSpace=BCS_BoneSpace
    translationMode=BMM_Additive
    translationSpace=BCS_BoneSpace
    alphaBoolBlend {blendOption=Linear}
    boneToModify {boneName=tricep_02_l}
    translation {x=-6}
   }
   n7 = "Transform (Modify) Bone - Bone: tricep_01_l"(ComponentPose=@n6, Alpha=@n3, Translation="-0.000,0.000,2.343", Rotation="0,0,0", Scale="1.000,1.500,1.250") {
    rotationMode=BMM_Additive
    rotationSpace=BCS_BoneSpace
    scaleMode=BMM_Additive
    scaleSpace=BCS_BoneSpace
    translationMode=BMM_Additive
    translationSpace=BCS_BoneSpace
    alphaBoolBlend {blendOption=Linear}
    boneToModify {boneName=tricep_01_l}
    scale {y=1.5, z=1.25}
    translation {z=2.343}
   }
   n8 = Get BicepRightLengthAlpha() -> (BicepRightLengthAlpha:float)
   n9 = "Transform (Modify) Bone - Bone: bicep_02_r"(ComponentPose=@n7, Alpha=@n8, Translation="-6,-0.000,-0.000", Rotation="0,0,0", Scale="1,1,1") {
    rotationMode=BMM_Additive
    rotationSpace=BCS_BoneSpace
    scaleMode=BMM_Additive
    scaleSpace=BCS_BoneSpace
    translationMode=BMM_Additive
    translationSpace=BCS_BoneSpace
    alphaBoolBlend {blendOption=Linear}
    boneToModify {boneName=bicep_02_r}
    translation {x=-6}
   }
   n10 = "Transform (Modify) Bone - Bone: bicep_01_r"(ComponentPose=@n9, Alpha=@n8, Translation="-0.000,0.000,2.343", Rotation="0,0,0", Scale="1.000,1.500,1.250") {
    rotationMode=BMM_Additive
    rotationSpace=BCS_BoneSpace
    scaleMode=BMM_Additive
    scaleSpace=BCS_BoneSpace
    translationMode=BMM_Additive
    translationSpace=BCS_BoneSpace
    alphaBoolBlend {blendOption=Linear}
    boneToModify {boneName=bicep_01_r}
    scale {y=1.5, z=1.25}
    translation {z=2.343}
   }
   n11 = "Transform (Modify) Bone - Bone: tricep_02_r"(ComponentPose=@n10, Alpha=@n8, Translation="-6,-0.000,-0.000", Rotation="0,0,0", Scale="1,1,1") {
    rotationMode=BMM_Additive
    rotationSpace=BCS_BoneSpace
    scaleMode=BMM_Additive
    scaleSpace=BCS_BoneSpace
    translationMode=BMM_Additive
    translationSpace=BCS_BoneSpace
    alphaBoolBlend {blendOption=Linear}
    alphaScaleBiasClamp {scale=0.5}
    boneToModify {boneName=tricep_02_r}
    translation {x=-6}
   }
   n12 = "Transform (Modify) Bone - Bone: tricep_01_r"(ComponentPose=@n11, Alpha=@n8, Translation="-0.000,0.000,2.343", Rotation="0,0,0", Scale="1.000,1.500,1.250") {
    rotationMode=BMM_Additive
    rotationSpace=BCS_BoneSpace
    scaleMode=BMM_Additive
    scaleSpace=BCS_BoneSpace
    translationMode=BMM_Additive
    translationSpace=BCS_BoneSpace
    alphaBoolBlend {blendOption=Linear}
    alphaScaleBiasClamp {scale=0.5}
    boneToModify {boneName=tricep_01_r}
    scale {y=1.5, z=1.25}
    translation {z=2.343}
   }
   Component To Local(ComponentPose=@n12)
   n13 = "Constraint - scapula_base_l (P)"(ComponentPose=Local To Component(), Alpha=1, ConstraintWeights_0=1) {
    alphaBoolBlend {blendOption=Linear}
    boneToModify {boneName=scapula_base_l}
    constraintSetup [{targetBone {boneName=clavicle_l}}]
    constraintWeights [1]
   }
   n14 = Look At - Bone: scapula_00_l(ComponentPose=@n13, Alpha=1) {
    alphaBoolBlend {blendOption=Linear}
    boneToModify {boneName=scapula_00_l}
    lookAtLocation {x=0}
    lookAtTarget {boneReference {boneName=scapula_01_l_aim}}
    lookAt_Axis {axis=<1 0>}
    lookUp_Axis {axis {x=0, z=1}}
   }
   n15 = Look At - Bone: pectoralis_01_l(ComponentPose=@n14, Alpha=1) {
    alphaBoolBlend {blendOption=Linear}
    boneToModify {boneName=pectoralis_01_l}
    lookAtLocation {x=0}
    lookAtTarget {boneReference {boneName=upperarm_l}}
    lookAt_Axis {axis=<1 0>}
    lookUp_Axis {axis {x=0, z=1}}
   }
   n16 = Look At - Bone: lattissimus_00_l(ComponentPose=@n15, Alpha=1) {
    alphaBoolBlend {blendOption=Linear}
    boneToModify {boneName=lattissimus_00_l}
    lookAtLocation {x=0}
    lookAtTarget {boneReference {boneName=upperarm_l}}
    lookAt_Axis {axis=<1 0>}
    lookUp_Axis {axis {x=0, z=1}}
   }
   n17 = "Constraint - scapula_base_r (P)"(ComponentPose=@n16, Alpha=1, ConstraintWeights_0=1) {
    alphaBoolBlend {blendOption=Linear}
    boneToModify {boneName=scapula_base_r}
    constraintSetup [{targetBone {boneName=clavicle_r}}]
    constraintWeights [1]
   }
   n18 = Look At - Bone: scapula_00_r(ComponentPose=@n17, Alpha=1) {
    alphaBoolBlend {blendOption=Linear}
    boneToModify {boneName=scapula_00_r}
    lookAtLocation {x=0}
    lookAtTarget {boneReference {boneName=scapula_01_r_aim}}
    lookAt_Axis {axis=<1 0>}
    lookUp_Axis {axis {x=0, z=1}}
   }
   n19 = Look At - Bone: pectoralis_01_r(ComponentPose=@n18, Alpha=1) {
    alphaBoolBlend {blendOption=Linear}
    boneToModify {boneName=pectoralis_01_r}
    lookAtLocation {x=0}
    lookAtTarget {boneReference {boneName=upperarm_r}}
    lookAt_Axis {axis=<1 0>}
    lookUp_Axis {axis {x=0, z=1}}
   }
   Look At - Bone: lattissimus_00_r(ComponentPose=@n19, Alpha=1) {
    alphaBoolBlend {blendOption=Linear}
    boneToModify {boneName=lattissimus_00_r}
    lookAtLocation {x=0}
    lookAtTarget {boneReference {boneName=upperarm_r}}
    lookAt_Axis {axis=<1 0>}
    lookUp_Axis {axis {x=0, z=1}}
   }
   n20 = Get RbicepRollAlpha() -> (RbicepRollAlpha:double)
   n21 = "Transform (Modify) Bone - Bone: bicep_bt_00_r"(Alpha=@n20, Translation="0,0,0", Rotation="0.000,0.000,-25.000", Scale="1,1,1") {
    rotationMode=BMM_Additive
    rotationSpace=BCS_BoneSpace
    translationMode=BMM_Additive
    translationSpace=BCS_BoneSpace
    alphaBoolBlend {blendOption=Linear}
    boneToModify {boneName=bicep_bt_00_r}
    rotation {roll=-24.9997}
   }
   "Transform (Modify) Bone - Bone: bicep_tp_00_r"(ComponentPose=@n21, Alpha=@n20, Translation="0,0,0", Rotation="0.000,0.000,-115.000", Scale="1,1,1") {
    rotationMode=BMM_Additive
    rotationSpace=BCS_BoneSpace
    translationMode=BMM_Additive
    translationSpace=BCS_BoneSpace
    alphaBoolBlend {blendOption=Linear}
    boneToModify {boneName=bicep_tp_00_r}
    rotation {roll=-114.9998}
   }
   n22 = "Constraint - pectoralis_02_r_aim (PPPPPPPPPPPPPPPPPP)"!(
    ComponentPose=Local To Component(LocalPose=Use cached pose 'initCache'())
    Alpha=1
    ConstraintWeights_0=0
    ConstraintWeights_1:float=0.0416
    ConstraintWeights_2:float=0.0394
    ConstraintWeights_3:float=0.0003
    ConstraintWeights_4:float=0
    ConstraintWeights_5:float=0
    ConstraintWeights_6:float=0.0122
    ConstraintWeights_7:float=0.259
    ConstraintWeights_8:float=0.0183
    ConstraintWeights_9:float=0.0001
    ConstraintWeights_10:float=0.0106
    ConstraintWeights_11:float=0
    ConstraintWeights_12:float=0.1954
    ConstraintWeights_13:float=0
    ConstraintWeights_14:float=0
    ConstraintWeights_15:float=0
    ConstraintWeights_16:float=0.4228
    ConstraintWeights_17:float=0.0003
   ) -> (Pose) {
    alphaBoolBlend {blendOption=Linear}
    boneToModify {boneName=pectoralis_02_r_aim}
    constraintSetup [
     {targetBone {boneName=tricep_01_r}}
     {targetBone {boneName=upperarm_r}}
     {targetBone {boneName=upperarm_twist_01_r}}
     {targetBone {boneName=deltoid_bk_01_r}}
     {targetBone {boneName=lattissimus_01_r}}
     {targetBone {boneName=pectoralis_02_r}}
     {targetBone {boneName=upperarm_twist_02_r}}
     {targetBone {boneName=deltoid_01_r}}
     {targetBone {boneName=spine_03}}
     {targetBone {boneName=spine_04}}
     {targetBone {boneName=spine_05}}
     {targetBone {boneName=tricep_02_r}}
     {targetBone {boneName=clavicle_r}}
     {targetBone {boneName=neck_02}}
     {targetBone {boneName=neck_01}}
     {targetBone {boneName=bicep_02_r}}
     {targetBone {boneName=deltoid_fr_01_r}}
     {targetBone {boneName=bicep_01_r}}
    ]
    constraintWeights [1,1,1,1,1,1,1,1,1,1,1,1,1,1,1,1,1,1]
   }
   n23 = "Constraint - pectoralis_02_l_aim (PPPPPPPPPPPPPPPPPPPPPPPPPP)"!(
    ComponentPose=@n22
    Alpha=1
    ConstraintWeights_0=0.0003
    ConstraintWeights_1:float=0.0122
    ConstraintWeights_2:float=0
    ConstraintWeights_3:float=0.0183
    ConstraintWeights_4:float=0.0001
    ConstraintWeights_5:float=0.0106
    ConstraintWeights_6:float=0
    ConstraintWeights_7:float=0.4228
    ConstraintWeights_8:float=0.1954
    ConstraintWeights_9:float=0.0394
    ConstraintWeights_10:float=0
    ConstraintWeights_11:float=0
    ConstraintWeights_12:float=0
    ConstraintWeights_13:float=0.0416
    ConstraintWeights_14:float=0
    ConstraintWeights_15:float=0
    ConstraintWeights_16:float=0.259
    ConstraintWeights_17:float=0
    ConstraintWeights_18:float=0
    ConstraintWeights_19:float=0
    ConstraintWeights_20:float=0
    ConstraintWeights_21:float=0.0003
    ConstraintWeights_22:float=0
    ConstraintWeights_23:float=0
    ConstraintWeights_24:float=0
    ConstraintWeights_25:float=0
   ) -> (Pose) {
    alphaBoolBlend {blendOption=Linear}
    boneToModify {boneName=pectoralis_02_l_aim}
    constraintSetup [
     {targetBone {boneName=deltoid_bk_01_l}}
     {targetBone {boneName=upperarm_twist_02_l}}
     {targetBone {boneName=spine_02}}
     {targetBone {boneName=spine_03}}
     {targetBone {boneName=spine_04}}
     {targetBone {boneName=spine_05}}
     {targetBone {boneName=pelvis}}
     {targetBone {boneName=deltoid_fr_01_l}}
     {targetBone {boneName=clavicle_l}}
     {targetBone {boneName=upperarm_twist_01_l}}
     {targetBone {boneName=clavicle_r}}
     {targetBone {boneName=neck_02}}
     {targetBone {boneName=neck_01}}
     {targetBone {boneName=upperarm_l}}
     {targetBone {boneName=lowerarm_twist_02_l}}
     {targetBone {boneName=head}}
     {targetBone {boneName=deltoid_01_l}}
     {targetBone {boneName=lattissimus_01_l}}
     {targetBone {boneName=bicep_02_l}}
     {targetBone {boneName=spine_01}}
     {targetBone {boneName=tricep_02_l}}
     {targetBone {boneName=bicep_01_l}}
     {targetBone {boneName=scapula_01_l}}
     {targetBone {boneName=pectoralis_02_l}}
     {targetBone {boneName=lowerarm_l}}
     {targetBone {boneName=tricep_01_l}}
    ]
    constraintWeights [1,1,1,1,1,1,1,1,1,1,1,1,1,1,1,1,1,1,1,1,1,1,1,1,1,1]
   }
   n24 = Look At - Bone: pectoralis_01_l!(ComponentPose=@n23, Alpha=1, LookAtLocation:Vector="0,0,0") -> (Pose) {
    alphaBoolBlend {blendOption=Linear}
    boneToModify {boneName=pectoralis_01_l}
    lookAtLocation {x=0}
    lookAtTarget {boneReference {boneName=pectoralis_02_l_aim}}
    lookAt_Axis {axis=<1 0>}
    lookUp_Axis {axis {x=0, z=1}}
   }
   n25 = Look At - Bone: pectoralis_01_r!(ComponentPose=@n24, Alpha=1, LookAtLocation:Vector="0,0,0") -> (Pose) {
    alphaBoolBlend {blendOption=Linear}
    boneToModify {boneName=pectoralis_01_r}
    lookAtLocation {x=0}
    lookAtTarget {boneReference {boneName=pectoralis_02_r_aim}}
    lookAt_Axis {axis=<1 0>}
    lookUp_Axis {axis {x=0, z=1}}
   }
   n26 = "Constraint - scapula_base_l (P)"(ComponentPose=@n25, Alpha=1, ConstraintWeights_0=1) {
    alphaBoolBlend {blendOption=Linear}
    boneToModify {boneName=scapula_base_l}
    constraintSetup [{targetBone {boneName=clavicle_l}}]
    constraintWeights [1]
   }
   n27 = Look At - Bone: scapula_00_l(ComponentPose=@n26, Alpha=1) {
    alphaBoolBlend {blendOption=Linear}
    boneToModify {boneName=scapula_00_l}
    lookAtLocation {x=0}
    lookAtTarget {boneReference {boneName=scapula_01_l_aim}}
    lookAt_Axis {axis=<1 0>}
    lookUp_Axis {axis {x=0, z=1}}
   }
   n28 = Look At - Bone: lattissimus_00_l(ComponentPose=@n27, Alpha=1) {
    alphaBoolBlend {blendOption=Linear}
    boneToModify {boneName=lattissimus_00_l}
    lookAtLocation {x=0}
    lookAtTarget {boneReference {boneName=upperarm_l}}
    lookAt_Axis {axis=<1 0>}
    lookUp_Axis {axis {x=0, z=1}}
   }
   n29 = "Constraint - scapula_base_r (P)"(ComponentPose=@n28, Alpha=1, ConstraintWeights_0=1) {
    alphaBoolBlend {blendOption=Linear}
    boneToModify {boneName=scapula_base_r}
    constraintSetup [{targetBone {boneName=clavicle_r}}]
    constraintWeights [1]
   }
   n30 = Look At - Bone: scapula_00_r(ComponentPose=@n29, Alpha=1) {
    alphaBoolBlend {blendOption=Linear}
    boneToModify {boneName=scapula_00_r}
    lookAtLocation {x=0}
    lookAtTarget {boneReference {boneName=scapula_01_r_aim}}
    lookAt_Axis {axis=<1 0>}
    lookUp_Axis {axis {x=0, z=1}}
   }
   n31 = Look At - Bone: lattissimus_00_r(ComponentPose=@n30, Alpha=1) {
    alphaBoolBlend {blendOption=Linear}
    boneToModify {boneName=lattissimus_00_r}
    lookAtLocation {x=0}
    lookAtTarget {boneReference {boneName=upperarm_r}}
    lookAt_Axis {axis=<1 0>}
    lookUp_Axis {axis {x=0, z=1}}
   }
   n32 = Get BicepLeftLengthAlpha()
   n33 = "Transform (Modify) Bone - Bone: bicep_02_l"(ComponentPose=@n31, Alpha=@n32, Translation="-6,-0.000,-0.000", Rotation="0,0,0", Scale="1,1,1") {
    rotationMode=BMM_Additive
    rotationSpace=BCS_BoneSpace
    scaleMode=BMM_Additive
    scaleSpace=BCS_BoneSpace
    translationMode=BMM_Additive
    translationSpace=BCS_BoneSpace
    alphaBoolBlend {blendOption=Linear}
    boneToModify {boneName=bicep_02_l}
    translation {x=-6}
   }
   n34 = "Transform (Modify) Bone - Bone: bicep_01_l"(ComponentPose=@n33, Alpha=@n32, Translation="-0.000,0.000,2.343", Rotation="0,0,0", Scale="1.000,1.500,1.250") {
    rotationMode=BMM_Additive
    rotationSpace=BCS_BoneSpace
    scaleMode=BMM_Additive
    scaleSpace=BCS_BoneSpace
    translationMode=BMM_Additive
    translationSpace=BCS_BoneSpace
    alphaBoolBlend {blendOption=Linear}
    boneToModify {boneName=bicep_01_l}
    scale {y=1.5, z=1.25}
    translation {z=2.343}
   }
   n35 = "Transform (Modify) Bone - Bone: tricep_02_l"(ComponentPose=@n34, Alpha=@n32, Translation="-6,-0.000,-0.000", Rotation="0,0,0", Scale="1,1,1") {
    rotationMode=BMM_Additive
    rotationSpace=BCS_BoneSpace
    scaleMode=BMM_Additive
    scaleSpace=BCS_BoneSpace
    translationMode=BMM_Additive
    translationSpace=BCS_BoneSpace
    alphaBoolBlend {blendOption=Linear}
    boneToModify {boneName=tricep_02_l}
    translation {x=-6}
   }
   n36 = "Transform (Modify) Bone - Bone: tricep_01_l"(ComponentPose=@n35, Alpha=@n32, Translation="-0.000,0.000,2.343", Rotation="0,0,0", Scale="1.000,1.500,1.250") {
    rotationMode=BMM_Additive
    rotationSpace=BCS_BoneSpace
    scaleMode=BMM_Additive
    scaleSpace=BCS_BoneSpace
    translationMode=BMM_Additive
    translationSpace=BCS_BoneSpace
    alphaBoolBlend {blendOption=Linear}
    alphaScaleBiasClamp {scale=0.3}
    boneToModify {boneName=tricep_01_l}
    scale {y=1.5, z=1.25}
    translation {z=2.343}
   }
   n37 = Get BicepRightLengthAlpha()
   n38 = "Transform (Modify) Bone - Bone: bicep_02_r"(ComponentPose=@n36, Alpha=@n37, Translation="-6,-0.000,-0.000", Rotation="0,0,0", Scale="1,1,1") {
    rotationMode=BMM_Additive
    rotationSpace=BCS_BoneSpace
    scaleMode=BMM_Additive
    scaleSpace=BCS_BoneSpace
    translationMode=BMM_Additive
    translationSpace=BCS_BoneSpace
    alphaBoolBlend {blendOption=Linear}
    boneToModify {boneName=bicep_02_r}
    translation {x=-6}
   }
   n39 = "Transform (Modify) Bone - Bone: bicep_01_r"(ComponentPose=@n38, Alpha=@n37, Translation="-0.000,0.000,2.343", Rotation="0,0,0", Scale="1.000,1.500,1.250") {
    rotationMode=BMM_Additive
    rotationSpace=BCS_BoneSpace
    scaleMode=BMM_Additive
    scaleSpace=BCS_BoneSpace
    translationMode=BMM_Additive
    translationSpace=BCS_BoneSpace
    alphaBoolBlend {blendOption=Linear}
    boneToModify {boneName=bicep_01_r}
    scale {y=1.5, z=1.25}
    translation {z=2.343}
   }
   n40 = "Transform (Modify) Bone - Bone: tricep_02_r"(ComponentPose=@n39, Alpha=@n37, Translation="-6,-0.000,-0.000", Rotation="0,0,0", Scale="1,1,1") {
    rotationMode=BMM_Additive
    rotationSpace=BCS_BoneSpace
    scaleMode=BMM_Additive
    scaleSpace=BCS_BoneSpace
    translationMode=BMM_Additive
    translationSpace=BCS_BoneSpace
    alphaBoolBlend {blendOption=Linear}
    alphaScaleBiasClamp {scale=0.5}
    boneToModify {boneName=tricep_02_r}
    translation {x=-6}
   }
   n41 = "Transform (Modify) Bone - Bone: tricep_01_r"(ComponentPose=@n40, Alpha=@n37, Translation="-0.000,0.000,2.343", Rotation="0,0,0", Scale="1.000,1.500,1.250") {
    rotationMode=BMM_Additive
    rotationSpace=BCS_BoneSpace
    scaleMode=BMM_Additive
    scaleSpace=BCS_BoneSpace
    translationMode=BMM_Additive
    translationSpace=BCS_BoneSpace
    alphaBoolBlend {blendOption=Linear}
    alphaScaleBiasClamp {scale=0.3}
    boneToModify {boneName=tricep_01_r}
    scale {y=1.5, z=1.25}
    translation {z=2.343}
   }
   n42 = RigidBody!(ComponentPose=@n41, Alpha, OverrideWorldGravity:Vector="0,0,-980", ExternalForce) -> (Pose) {
    bClampLinearTranslationLimitToRefPose=true
    bOverrideWorldGravity=true
    overridePhysicsAsset=/Game/Characters/Player/Male/Medium/Bodies/M_MED_Sandcastle/Meshes/M_MED_Sandcastle_Physics
    simulationSpace=BaseBoneSpace
    alphaBoolBlend {blendOption=Linear}
    baseBoneRef {boneName=pelvis}
    componentAppliedLinearAccClamp=<3000 3000 3000>
    componentLinearAccScale=<0.3 0.3 0.3>
    componentLinearVelScale=<0.3 0.3 0.3>
    overrideWorldGravity {z=-980}
   }
   Output Pose(Result=Component To Local(ComponentPose=@n42)) {layerGroup=None}
   Save cached pose 'initCache'(Pose=Copy Pose From Mesh() {bUseAttachedParent=true}) {cacheName=initCache}
   Use cached pose 'initCache'()
  ]}
 EventGraph {links [], nodes []}
]
variables [RbicepRollAlpha {type=number}]
